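# Translucent unlit Niagara material for the laser circle effect in the X-rays vehicle boost cosmetic.
# Fresnel blends two colors, scaled by the Glow dynamic parameter into an emissive adjustment function.
# A sine of a radial gradient shapes rings, multiplied by a radially mapped, panning fire texture, then smoothstepped by Dissolve and faded by particle alpha.
# Scrolling sine waves on the second UV channel build a two-axis distortion vector scaled by Distort.
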Material /VehicleCosmetics/Boosts/Boost_Xrays/Materials/M_LaserCircles
graph {
 properties {
  bAutomaticallySetUsageInEditor=false
  bUsedWithNiagaraMeshParticles=true
  bUsedWithNiagaraRibbons=true
  bUsedWithNiagaraSprites=true
  blendMode=BLEND_Translucent
  shadingModel=MSM_Unlit
  usedShadingModels=Unlit
 }
 nodes [
  n1 = VectorParameter() {parameterName=Color02, defaultValue=<rgba 1 0.6936 0.8197 1>}
  n2 = VectorParameter() {parameterName=Color01, defaultValue=<rgba 0.9573 0 0.02 1>}
  n3 = Divide(
   A=MaterialFunctionCall(
    Power=ScalarParameter() {defaultValue=1, parameterName=Power}
    Use Cheap contrast=StaticBool() {value=true}
    Cheap contrast dark=ScalarParameter() {defaultValue=1, parameterName=Dark}
    Cheap contrast bright=ScalarParameter() {defaultValue=1, parameterName=bright}
   ) {materialFunction=&Fresnel_Function}.Result
   B=ScalarParameter() {defaultValue=0.01, parameterName=Divide}
  )
  MaterialExpressionMeshEmitterDynamicParameter_2 = DynamicParameter() {paramNames [Dissolve,Glow,Distort,Param4]}
  MaterialFunctionCall(Emissive=Multiply(A=LinearInterpolate(A=@n1.RGB, B=@n2.RGB, Alpha=@n3), B=@MaterialExpressionMeshEmitterDynamicParameter_2.Glow)) {materialFunction=&MF_FX_EmissiveAdjustments}
  n4 = Time()
  n5 = Divide(A=ComponentMask(None=TextureCoordinate() {coordinateIndex=1}) {r=true}, B=ScalarParameter() {defaultValue=0.8, parameterName=CountWaves2})
  n6 = Add(A=@n5, B=Multiply(A=ScalarParameter() {defaultValue=0.5, parameterName=HorizontalSpeed}, B=OneMinus(None=@n4)))
  n7 = Time()
  n8 = Add(A=@n5, B=Multiply(A=ScalarParameter() {defaultValue=-0.2, parameterName=HorizontalSpeed2}, B=OneMinus(None=@n7)))
  n9 = Power(Base=Multiply(A=Sine(None=@n6), B=Sine(None=@n8)), Exp=Constant() {r=1.2})
  n10 = Divide(A=ComponentMask(None=TextureCoordinate() {coordinateIndex=1}) {g=true}, B=ScalarParameter() {defaultValue=0.8, parameterName=CountWaves2})
  n11 = ScalarParameter() {defaultValue=1, parameterName=VerticalSpeed}
  n12 = OneMinus(None=Time())
  n13 = Sine(None=Add(A=@n10, B=Multiply(A=@n11, B=@n12)))
  n14 = ScalarParameter() {defaultValue=-0.5, parameterName=VerticalSpeed2}
  n15 = OneMinus(None=Time())
  n16 = Sine(None=Add(A=@n10, B=Multiply(A=@n14, B=@n15)))
  n17 = Power(Base=Multiply(A=OneMinus(None=@n13), B=OneMinus(None=@n16)), Exp=Constant() {r=1.2})
  Reroute(None=Multiply(A=AppendVector(A=@n9, B=@n17), B=@MaterialExpressionMeshEmitterDynamicParameter_2.Distort))
  MaterialExpressionMeshEmitterVertexColor_2 = ParticleColor()
  n18 = ScalarParameter() {parameterName=Max}
  n19 = Sine(None=MaterialFunctionCall(Radius=ScalarParameter() {defaultValue=0.5, parameterName=Radius}, Density=ScalarParameter() {defaultValue=1, parameterName=Density}) {materialFunction=&RadialGradientExponential}.RadialGradientExponential) {period=2}
  n20 = MaterialFunctionCall() {materialFunction=&VectorToRadialValue}
  n21 = AppendVector(A=ScalarParameter() {defaultValue=1, parameterName=TileX}, B=ScalarParameter() {defaultValue=1, parameterName=TileY})
  n22 = TextureSample(UVs=Panner(Coordinate=Multiply(A=@n20.Radial Coordinates, B=@n21)) {speedY=-0.3}) {texture=/Game/Effects/Fort_Effects/Textures/T_Dissolve_Fire_M}
  MaterialExpressionConstantClamp_6 = Saturate(None=SmoothStep(
    Min=@MaterialExpressionMeshEmitterDynamicParameter_2.Dissolve
    Max=Add(A=@MaterialExpressionMeshEmitterDynamicParameter_2.Dissolve, B=@n18)
    Value=Multiply(A=@n19, B=@n22.RGB)
   ))
  MaterialExpressionConstantClamp_5 = Saturate(None=Multiply(A=@MaterialExpressionMeshEmitterVertexColor_2.A, B=@MaterialExpressionConstantClamp_6))
  Reroute(None=@MaterialExpressionConstantClamp_5)
 ]
}
