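# Translucent unlit ribbon trail material for the X-rays vehicle boost effect, usable on Niagara ribbons, sprites and mesh particles.
# A panned cloud texture, scaled by dynamic and scalar Distort parameters, offsets the UVs of a scrolling trail texture.
# Opacity combines that texture with particle alpha, power-curve fades at both U ends, gain and a depth fade function.
# Particle color feeds an emissive adjustment function.
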Material /VehicleCosmetics/Boosts/Boost_Xrays/Materials/M_RibbonLineTranslucent
graph {
 properties {
  bAutomaticallySetUsageInEditor=false
  bUsedWithMeshParticles=true
  bUsedWithNiagaraRibbons=true
  bUsedWithNiagaraSprites=true
  blendMode=BLEND_Translucent
  shadingModel=MSM_Unlit
  usedShadingModels=Unlit
 }
 nodes [
  n1 = Panner(Coordinate=TextureCoordinate() {uTiling=1.5, vTiling=0.5}, Time=Multiply(A=Time(), B=Constant() {r=1})) {speedX=-1, speedY=0.2}
  n2 = Multiply(
   A=ComponentMask(None=TextureSample(UVs=@n1) {texture=/VehicleCosmetics/Boosts/Boost_ToonyWater/Texture/T_Boost_ToonyWater_Clouds_RGB_Soft_T}.RGB) {g=true, r=true}
   B=DynamicParameter() {paramNames [Distort,Param2,Param3,Param4]}.Distort
  )
  n3 = ScalarParameter() {defaultValue=1, parameterName=Distort}
  n4 = TextureCoordinate()
  n5 = TextureSampleParameter2D(UVs=Add(A=Multiply(A=@n2, B=@n3), B=Panner(Coordinate=@n4) {speedX=-1.3})) {parameterName=Texture, texture=/VehicleCosmetics/Boosts/Boost_Xrays/Textures/TrailMain07_T}
  MaterialExpressionVertexColor_0 = ParticleColor()
  n6 = Multiply(A=ComponentMask(None=TextureCoordinate()) {r=true}, B=Constant() {r=1})
  n7 = ScalarParameter() {defaultValue=1, parameterName=power fade}
  n8 = ComponentMask(None=TextureCoordinate()) {r=true}
  n9 = Multiply(
   A=Multiply(A=Multiply(A=@n5.RGB, B=@MaterialExpressionVertexColor_0.A), B=Power(Base=@n6, Exp=@n7))
   B=Power(Base=OneMinus(None=@n8), Exp=ScalarParameter() {defaultValue=0.8, parameterName=Powerback fade})
  )
  n10 = ScalarParameter() {defaultValue=2, parameterName=Opacity_Gain}
  n11 = ScalarParameter() {defaultValue=100, parameterName=Fade_Length}
  n12 = ScalarParameter() {defaultValue=16, parameterName=Fade_Offset}
  MaterialExpressionConstantClamp_0 = Saturate(None=Multiply(
    A=Multiply(A=@n9, B=@n10)
    B=MaterialFunctionCall(Opacity In=@n11, DepthFade Distance=@n12) {materialFunction=&MF_VCVFX_DepthFade}.Result
   ))
  DynamicParameter() {parameterIndex=1, paramNames [Distort,Param2,Param3,Param4]}
  Reroute(None=MaterialFunctionCall(Emissive=@MaterialExpressionVertexColor_0.RGB) {materialFunction=&MF_FX_EmissiveAdjustments}.Result)
 ]
}
