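# Translucent unlit Niagara material for the inner-flame trail of the Xrays vehicle boost cosmetic.
# A panning cloud texture borrowed from the ToonyWater boost distorts the UVs of a fast-scrolling trail texture.
# The trail's red channel, tinted by particle color, feeds MF_FX_EmissiveAdjustments for glow.
# A separate mask multiplies the trail by particle alpha, a fade along U and an exponential radial gradient to shape the flame.
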Material /VehicleCosmetics/Boosts/Boost_Xrays/Materials/M_Trail_InnerFlames_Translucent
graph {
 properties {
  bAutomaticallySetUsageInEditor=false
  bUsedWithNiagaraMeshParticles=true
  bUsedWithNiagaraRibbons=true
  bUsedWithNiagaraSprites=true
  blendMode=BLEND_Translucent
  shadingModel=MSM_Unlit
  usedShadingModels=Unlit
 }
 nodes [
  n1 = TextureCoordinate() {uTiling=0.35, vTiling=0.35}
  n2 = Multiply(A=Time(), B=Constant() {r=5})
  n3 = ComponentMask(None=TextureSample(UVs=Panner(Coordinate=@n1, Time=@n2) {speedX=0.2, speedY=0.2}) {texture=/VehicleCosmetics/Boosts/Boost_ToonyWater/Texture/T_Boost_ToonyWater_Clouds_RGB_Soft_T}.RGB) {g=true, r=true}
  n4 = ScalarParameter() {defaultValue=0.02, parameterName=Distortion}
  n5 = TextureCoordinate()
  n6 = TextureSample(UVs=Add(A=Multiply(A=@n3, B=@n4), B=Panner(Coordinate=@n5) {speedX=-1.8})) {texture=/VehicleCosmetics/Boosts/Boost_Xrays/Textures/TrailMain05_T}
  MaterialExpressionVertexColor_1 = ParticleColor()
  n7 = Multiply(A=@n6.RGB, B=@MaterialExpressionVertexColor_1.A)
  n8 = TextureCoordinate()
  n9 = Power(Base=OneMinus(None=ComponentMask(None=@n8) {r=true}), Exp=Constant() {r=0.5})
  n10 = MaterialFunctionCall(UVs=TextureCoordinate(), CenterPosition=Constant2Vector() {g=0.5, r=0.6}, Radius=Constant() {r=0.6}, Density=Constant() {r=0.9}) {materialFunction=&RadialGradientExponential}
  n11 = Constant() {r=1.45}
  Multiply(A=Multiply(A=Multiply(A=@n7, B=@n9), B=Power(Base=@n10.RadialGradientExponential, Exp=@n11)), B=Constant() {r=2})
  Reroute(None=MaterialFunctionCall(Emissive=Multiply(A=@MaterialExpressionVertexColor_1.RGB, B=@n6.R)) {materialFunction=&MF_FX_EmissiveAdjustments}.Result)
 ]
}
